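# Physics asset for the HazardEdge male medium body's belt, giving its dangling parts secondary motion against the character.
# Back, front and ring chains start at zero-size base bodies on the pelvis, with limited swing and tight twist along each chain.
# Both belt ends join the last ring through free-angular constraints, closing a loop, and two compound-shape dyn_bone pieces hang from that ring.
# Pelvis, spine_03 and left-leg capsules provide body collision shapes.
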PhysicsAsset /BRCosmetics/Characters/Player/Male/Medium/Bodies/M_MED_HazardEdge/Meshes/M_MED_HazardEdge_Physics_Belt
bodies {
 calf_l [{shapeType=Capsule, length=35.1206, radius=6.5392, shapeName=calf_l_capsule, center=<-20.5762 -2.6636 0.7306>, rotation=<rot 90 180 179.1029>}]
 dyn_belt_bk_01 [{
   shapeType=Capsule
   length=5.6187
   radius=2.1408
   shapeName=dyn_belt_bk_01_capsule
   center=<4.8316 -0.3686 0.2023>
   rotation=<rot 83.4095 -179.9971 162.5257>
  }]
 dyn_belt_bk_02 [{
   shapeType=Capsule
   length=5.6494
   radius=2.1101
   shapeName=dyn_belt_bk_02_capsule
   center=<3.081 -1.3579 0>
   rotation=<rot 83.4094 -179.9986 -179.9986>
  }]
 dyn_belt_bk_03 [{
   shapeType=Capsule
   length=5.6999
   radius=2.0596
   shapeName=dyn_belt_bk_03_capsule
   center=<3.0329 -1.0031 -0.0056>
   rotation=<rot 83.4092 -179.9979 167.6732>
  }]
 dyn_belt_bk_base [{shapeType=Capsule, length=0, radius=0, shapeName=dyn_belt_bk_base_capsule}]
 dyn_belt_frnt_01 [{
   shapeType=Capsule
   length=5.2967
   radius=1.9568
   shapeName=dyn_belt_frnt_01_capsule
   center=<3.8471 0.3688 0.8573>
   rotation=<rot 83.6855 -179.9954 -169.1528>
  }]
 dyn_belt_frnt_02 [{shapeType=Capsule, length=4.436, radius=2.0199, shapeName=dyn_belt_frnt_02_capsule, center=<3.1219 0 0>, rotation=<rot 90 -63.5364 -40.6234>}]
 dyn_belt_frnt_base [{shapeType=Capsule, length=0, radius=0, shapeName=dyn_belt_frnt_base_capsule}]
 dyn_bone_01_base [
  {
   shapeType=Capsule
   length=3.2331
   radius=1.9066
   shapeName=dyn_bone_01_base_capsule
   center=<11.5279 -1.2136 -0.533>
   rotation=<rot 5.5084 168.0501 -56.8863>
  }
  {
   shapeType=Capsule
   length=4.0796
   radius=2.084
   shapeName=dyn_bone_01_base_capsule
   center=<4.0332 1.766 1.5655>
   rotation=<rot -12.3339 -12.449 61.1345>
  }
  {
   shapeType=Box
   extentX=7.2234
   extentY=2.771
   extentZ=1.6662
   shapeName=dyn_bone_01_base_box
   center=<7.0134 0.0946 -0.3666>
   rotation=<rot -10.5196 -17.5878 -30.7567>
  }
 ]
 dyn_bone_02_base [
  {
   shapeType=Capsule
   length=2.3696
   radius=2.0042
   shapeName=dyn_bone_02_base_capsule
   center=<11.2545 -1.509 -0.7451>
   rotation=<rot 6.5641 175.5269 -64.342>
  }
  {
   shapeType=Capsule
   length=2.9044
   radius=2.0042
   shapeName=dyn_bone_02_base_capsule
   center=<2.719 -1.4523 0.0698>
   rotation=<rot -1.7343 -1.7342 -124.2723>
  }
  {
   shapeType=Box
   extentX=7.5
   extentY=3.4872
   extentZ=1.3737
   shapeName=dyn_bone_02_base_box
   center=<6.5531 -0.9252 -1.2357>
   rotation=<rot -5.0133 0.0667 -30.7183>
  }
 ]
 dyn_rings_01 [{shapeType=Capsule, length=2.5873, radius=2.2796, shapeName=dyn_rings_01_capsule, center=<3.172 0 0>, rotation=<rot 90 0 0>}]
 dyn_rings_02 [{shapeType=Capsule, length=5.3065, radius=2.5517, shapeName=dyn_rings_02_capsule, center=<1.5915 0 2.0706>, rotation=<rot 0 0 90>}]
 dyn_rings_base [{shapeType=Capsule, length=0, radius=0, shapeName=dyn_rings_base_capsule}]
 foot_l [{shapeType=Capsule, length=12.1525, radius=6.8851, shapeName=foot_l_capsule, center=<-4.419 2.9673 0>, rotation=<rot 0 0 90>}]
 pelvis [{shapeType=Capsule, length=11.4235, radius=10.2307, shapeName=pelvis_capsule, center=<-4.3853 0 -0.0254>}]
 spine_03 [{shapeType=Capsule, length=17.6852, radius=16.3805, shapeName=spine_03_capsule, center=<-4.7112 8.3615 0>}]
 thigh_l [{
   shapeType=Capsule
   length=35.248
   radius=9.1756
   shapeName=thigh_l_capsule
   center=<-17.9725 0.3733 2.9799>
   rotation=<rot 83.7851 0.0001 -6.5817>
  }]
}
constraints [
 {
  bone1Name=dyn_belt_bk_base
  bone2Name=pelvis
  swing1LimitDegrees=45
  swing1Motion=Free
  swing2LimitDegrees=45
  swing2Motion=Free
  twistLimitDegrees=45
  twistMotion=Free
 }
 {bone1Name=dyn_belt_bk_01, bone2Name=dyn_belt_bk_base, swing1LimitDegrees=25, swing2LimitDegrees=25, twistLimitDegrees=5}
 {bone1Name=dyn_belt_bk_02, bone2Name=dyn_belt_bk_01, swing1LimitDegrees=25, swing2LimitDegrees=25, twistLimitDegrees=5}
 {bone1Name=dyn_belt_bk_03, bone2Name=dyn_belt_bk_02, swing1LimitDegrees=25, swing2LimitDegrees=25, twistLimitDegrees=5}
 {
  bone1Name=dyn_belt_frnt_base
  bone2Name=pelvis
  swing1LimitDegrees=45
  swing1Motion=Free
  swing2LimitDegrees=45
  swing2Motion=Free
  twistLimitDegrees=45
  twistMotion=Free
 }
 {bone1Name=dyn_belt_frnt_01, bone2Name=dyn_belt_frnt_base, swing1LimitDegrees=30, swing2LimitDegrees=25, twistLimitDegrees=5}
 {bone1Name=dyn_belt_frnt_02, bone2Name=dyn_belt_frnt_01, swing1LimitDegrees=30, swing2LimitDegrees=25, twistLimitDegrees=5}
 {
  bone1Name=dyn_rings_base
  bone2Name=pelvis
  swing1LimitDegrees=45
  swing1Motion=Free
  swing2LimitDegrees=45
  swing2Motion=Free
  twistLimitDegrees=45
  twistMotion=Free
 }
 {bone1Name=dyn_rings_01, bone2Name=dyn_rings_base, swing1LimitDegrees=5, swing2LimitDegrees=15, twistLimitDegrees=10}
 {bone1Name=dyn_rings_02, bone2Name=dyn_rings_01, swing1LimitDegrees=10, swing2LimitDegrees=10, twistLimitDegrees=10}
 {
  bone1Name=thigh_l
  bone2Name=pelvis
  swing1LimitDegrees=45
  swing1Motion=Free
  swing2LimitDegrees=45
  swing2Motion=Free
  twistLimitDegrees=45
  twistMotion=Free
 }
 {
  bone1Name=dyn_rings_02
  bone2Name=dyn_belt_bk_03
  swing1LimitDegrees=45
  swing1Motion=Free
  swing2LimitDegrees=45
  swing2Motion=Free
  twistLimitDegrees=45
  twistMotion=Free
 }
 {
  bone1Name=dyn_rings_02
  bone2Name=dyn_belt_frnt_02
  swing1LimitDegrees=45
  swing1Motion=Free
  swing2LimitDegrees=45
  swing2Motion=Free
  twistLimitDegrees=45
  twistMotion=Free
 }
 {bone1Name=dyn_bone_01_base, bone2Name=dyn_rings_02, swing1LimitDegrees=45, swing2LimitDegrees=45, twistLimitDegrees=5}
 {bone1Name=dyn_bone_02_base, bone2Name=dyn_rings_02, swing1LimitDegrees=30.6, swing2LimitDegrees=30.6, twistLimitDegrees=10}
 {
  bone1Name=calf_l
  bone2Name=thigh_l
  swing1LimitDegrees=45
  swing1Motion=Free
  swing2LimitDegrees=45
  swing2Motion=Free
  twistLimitDegrees=45
  twistMotion=Free
 }
 {
  bone1Name=foot_l
  bone2Name=calf_l
  swing1LimitDegrees=45
  swing1Motion=Free
  swing2LimitDegrees=45
  swing2Motion=Free
  twistLimitDegrees=45
  twistMotion=Free
 }
 {
  bone1Name=spine_03
  bone2Name=pelvis
  swing1LimitDegrees=45
  swing1Motion=Free
  swing2LimitDegrees=45
  swing2Motion=Free
  twistLimitDegrees=45
  twistMotion=Free
 }
]
