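# Physics asset for jacket secondary motion on the HazardEdge medium male player body.
# Seven two-segment capsule chains (back left/middle/right, both sides, both fronts) hang from zero-size base bodies tied to spine_04 by free-rotation constraints.
# Segment joints have angular limits, tight on back and sides and looser on the fronts, whose upper segments use triple capsules for wider coverage.
# Pelvis, spine and thigh capsules supply the body volumes around the jacket.
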PhysicsAsset /BRCosmetics/Characters/Player/Male/Medium/Bodies/M_MED_HazardEdge/Meshes/M_MED_HazardEdge_Physics_Jacket
bodies {
 dyn_jacket_bk_l_01 [{shapeType=Capsule, length=6, radius=4.8048, shapeName=dyn_jacket_bk_l_01_capsule, center=<5.4725 0 0>}]
 dyn_jacket_bk_l_02 [{shapeType=Capsule, length=9.6618, radius=3.2624, shapeName=dyn_jacket_bk_l_02_capsule, center=<5.4725 0 0>}]
 dyn_jacket_bk_l_base [{shapeType=Capsule, length=0, radius=0, shapeName=dyn_jacket_bk_l_base_capsule}]
 dyn_jacket_bk_mid_01 [{shapeType=Capsule, length=15.452, radius=4.8048, shapeName=dyn_jacket_bk_mid_01_capsule, center=<5.692 0 0>}]
 dyn_jacket_bk_mid_02 [{shapeType=Capsule, length=10.9108, radius=4.789, shapeName=dyn_jacket_bk_mid_02_capsule, center=<5.692 0 0>}]
 dyn_jacket_bk_mid_base [{shapeType=Capsule, length=0, radius=0, shapeName=dyn_jacket_bk_mid_base_capsule}]
 dyn_jacket_bk_r_01 [{shapeType=Capsule, length=6, radius=4.8048, shapeName=dyn_jacket_bk_l_01_capsule, center=<-5.4725 0 0>, rotation=<rot 0 0 -180>}]
 dyn_jacket_bk_r_02 [{shapeType=Capsule, length=9.6618, radius=3.2624, shapeName=dyn_jacket_bk_l_02_capsule, center=<-5.4725 0 0>, rotation=<rot 0 0 -180>}]
 dyn_jacket_bk_r_base [{shapeType=Capsule, length=0, radius=0, shapeName=dyn_jacket_bk_r_base_capsule}]
 dyn_jacket_frnt_l_01 [
  {shapeType=Capsule, length=11.5634, radius=4.2114, shapeName=dyn_jacket_frnt_l_01_capsule, center=<6.9665 1.7599 0>, rotation=<rot 90 0 0>}
  {
   shapeType=Capsule
   length=11.5167
   radius=4.2581
   shapeName=dyn_jacket_frnt_l_01_capsule
   center=<6.9665 1.7599 5.9036>
   rotation=<rot 90 0 0>
  }
  {
   shapeType=Capsule
   length=14.8613
   radius=4.2581
   shapeName=dyn_jacket_frnt_l_01_capsule
   center=<6.9665 1.329 -5.8145>
   rotation=<rot 90 0 0>
  }
 ]
 dyn_jacket_frnt_l_02 [{
   shapeType=Capsule
   length=14.349
   radius=4.5835
   shapeName=dyn_jacket_frnt_l_02_capsule
   center=<6.1535 1.7599 -0.3979>
   rotation=<rot 19.9998 0.0002 0>
  }]
 dyn_jacket_frnt_l_base [{shapeType=Capsule, length=0, radius=0, shapeName=dyn_jacket_frnt_l_base_capsule}]
 dyn_jacket_frnt_r_01 [
  {
   shapeType=Capsule
   length=11.5634
   radius=4.2114
   shapeName=dyn_jacket_frnt_l_01_capsule
   center=<-6.9665 -1.7599 0>
   rotation=<rot 90 0 180>
  }
  {
   shapeType=Capsule
   length=11.5167
   radius=4.2581
   shapeName=dyn_jacket_frnt_l_01_capsule
   center=<-6.9665 -1.7599 -5.9036>
   rotation=<rot 90 0 180>
  }
  {
   shapeType=Capsule
   length=14.8613
   radius=4.2581
   shapeName=dyn_jacket_frnt_l_01_capsule
   center=<-6.9665 -1.329 5.8145>
   rotation=<rot 90 0 180>
  }
 ]
 dyn_jacket_frnt_r_02 [{
   shapeType=Capsule
   length=11.9189
   radius=4.6829
   shapeName=dyn_jacket_frnt_l_02_capsule
   center=<-7.9593 0 1.9865>
   rotation=<rot 19.9998 0.0002 180>
  }]
 dyn_jacket_frnt_r_base [{shapeType=Capsule, length=0, radius=0, shapeName=dyn_jacket_frnt_r_base_capsule}]
 dyn_jacket_side_l_01 [{shapeType=Capsule, length=8.6173, radius=5.1199, shapeName=dyn_jacket_side_l_01_capsule, center=<0.2359 0 1.6211>, rotation=<rot 0 0 90>}]
 dyn_jacket_side_l_02 [{shapeType=Capsule, length=11.7933, radius=3.7413, shapeName=dyn_jacket_side_l_02_capsule, center=<3.4996 0 0>, rotation=<rot 0 -9.8343 90>}]
 dyn_jacket_side_l_base [{shapeType=Capsule, length=0, radius=0, shapeName=dyn_jacket_side_l_base_capsule}]
 dyn_jacket_side_r_01 [{shapeType=Capsule, length=8.6173, radius=5.1199, shapeName=dyn_jacket_side_l_01_capsule, center=<-0.2359 0 0>, rotation=<rot 0 0 -90>}]
 dyn_jacket_side_r_02 [{shapeType=Capsule, length=11.7933, radius=3.7413, shapeName=dyn_jacket_side_l_02_capsule, center=<-3.4996 0 0>, rotation=<rot 0 -9.8343 -90>}]
 dyn_jacket_side_r_base [{shapeType=Capsule, length=0, radius=0, shapeName=dyn_jacket_side_r_base_capsule}]
 pelvis [{shapeType=Capsule, length=13.7701, radius=11.4221, shapeName=pelvis_capsule, center=<-3.2787 0 0>}]
 spine_01 [{shapeType=Capsule, length=9.6509, radius=9.5139, shapeName=spine_01_capsule, center=<-1.5455 2.5615 0>}]
 spine_02 [{shapeType=Capsule, length=10.7311, radius=8.712, shapeName=spine_02_capsule, center=<-4.3479 0 0>}]
 spine_03 [{shapeType=Capsule, length=9.0887, radius=11.5455, shapeName=spine_03_capsule, center=<-4.0514 1.3483 0>}]
 spine_04 [{shapeType=Capsule, length=12.7424, radius=10.1493, shapeName=spine_04_capsule, center=<-2.9632 1.7735 0>}]
 thigh_l [{shapeType=Capsule, length=24.169, radius=8.2924, shapeName=thigh_l_capsule, center=<-17.2019 0 0>, rotation=<rot 90 0 0>}]
 thigh_r [{shapeType=Capsule, length=24.169, radius=8.2924, shapeName=thigh_l_capsule, center=<17.2019 0 0>, rotation=<rot 90 0 180>}]
}
constraints [
 {
  bone1Name=spine_02
  bone2Name=spine_01
  swing1LimitDegrees=45
  swing1Motion=Free
  swing2LimitDegrees=45
  swing2Motion=Free
  twistLimitDegrees=45
  twistMotion=Free
 }
 {
  bone1Name=spine_03
  bone2Name=spine_02
  swing1LimitDegrees=45
  swing1Motion=Free
  swing2LimitDegrees=45
  swing2Motion=Free
  twistLimitDegrees=45
  twistMotion=Free
 }
 {
  bone1Name=spine_04
  bone2Name=spine_03
  swing1LimitDegrees=45
  swing1Motion=Free
  swing2LimitDegrees=45
  swing2Motion=Free
  twistLimitDegrees=45
  twistMotion=Free
 }
 {
  bone1Name=dyn_jacket_bk_l_base
  bone2Name=spine_04
  swing1LimitDegrees=45
  swing1Motion=Free
  swing2LimitDegrees=45
  swing2Motion=Free
  twistLimitDegrees=45
  twistMotion=Free
 }
 {bone1Name=dyn_jacket_bk_l_01, bone2Name=dyn_jacket_bk_l_base, swing1LimitDegrees=10.76, swing2LimitDegrees=2, twistLimitDegrees=0}
 {bone1Name=dyn_jacket_bk_l_02, bone2Name=dyn_jacket_bk_l_01, swing1LimitDegrees=5, swing2LimitDegrees=2, twistLimitDegrees=0}
 {
  bone1Name=dyn_jacket_bk_mid_base
  bone2Name=spine_04
  swing1LimitDegrees=45
  swing1Motion=Free
  swing2LimitDegrees=45
  swing2Motion=Free
  twistLimitDegrees=45
  twistMotion=Free
 }
 {bone1Name=dyn_jacket_bk_mid_01, bone2Name=dyn_jacket_bk_mid_base, swing1LimitDegrees=5, swing2LimitDegrees=2, twistLimitDegrees=0}
 {bone1Name=dyn_jacket_bk_mid_02, bone2Name=dyn_jacket_bk_mid_01, swing1LimitDegrees=5, swing2LimitDegrees=2, twistLimitDegrees=0}
 {
  bone1Name=dyn_jacket_bk_r_base
  bone2Name=spine_04
  swing1LimitDegrees=45
  swing1Motion=Free
  swing2LimitDegrees=45
  swing2Motion=Free
  twistLimitDegrees=45
  twistMotion=Free
 }
 {
  bone1Name=dyn_jacket_frnt_l_base
  bone2Name=spine_04
  swing1LimitDegrees=45
  swing1Motion=Free
  swing2LimitDegrees=45
  swing2Motion=Free
  twistLimitDegrees=45
  twistMotion=Free
 }
 {bone1Name=dyn_jacket_frnt_l_01, bone2Name=dyn_jacket_frnt_l_base, swing1LimitDegrees=25, swing2LimitDegrees=5, twistLimitDegrees=5}
 {bone1Name=dyn_jacket_frnt_l_02, bone2Name=dyn_jacket_frnt_l_01, swing1LimitDegrees=45, swing2LimitDegrees=5, twistLimitDegrees=10}
 {
  bone1Name=dyn_jacket_frnt_r_base
  bone2Name=spine_04
  swing1LimitDegrees=45
  swing1Motion=Free
  swing2LimitDegrees=45
  swing2Motion=Free
  twistLimitDegrees=45
  twistMotion=Free
 }
 {
  bone1Name=dyn_jacket_side_l_base
  bone2Name=spine_04
  swing1LimitDegrees=45
  swing1Motion=Free
  swing2LimitDegrees=45
  swing2Motion=Free
  twistLimitDegrees=45
  twistMotion=Free
 }
 {bone1Name=dyn_jacket_side_l_01, bone2Name=dyn_jacket_side_l_base, swing1LimitDegrees=2.8, swing2LimitDegrees=5, twistLimitDegrees=0}
 {bone1Name=dyn_jacket_side_l_02, bone2Name=dyn_jacket_side_l_01, swing1LimitDegrees=2.88, swing2LimitDegrees=5, twistLimitDegrees=0}
 {
  bone1Name=dyn_jacket_side_r_base
  bone2Name=spine_04
  swing1LimitDegrees=45
  swing1Motion=Free
  swing2LimitDegrees=45
  swing2Motion=Free
  twistLimitDegrees=45
  twistMotion=Free
 }
 {
  bone1Name=spine_01
  bone2Name=pelvis
  swing1LimitDegrees=45
  swing1Motion=Free
  swing2LimitDegrees=45
  swing2Motion=Free
  twistLimitDegrees=45
  twistMotion=Free
 }
 {
  bone1Name=thigh_l
  bone2Name=pelvis
  swing1LimitDegrees=45
  swing1Motion=Free
  swing2LimitDegrees=45
  swing2Motion=Free
  twistLimitDegrees=45
  twistMotion=Free
 }
 {bone1Name=dyn_jacket_bk_r_01, bone2Name=dyn_jacket_bk_r_base, swing1LimitDegrees=5, swing2LimitDegrees=2, twistLimitDegrees=0}
 {
  bone1Name=thigh_r
  bone2Name=pelvis
  swing1LimitDegrees=45
  swing1Motion=Free
  swing2LimitDegrees=45
  swing2Motion=Free
  twistLimitDegrees=45
  twistMotion=Free
 }
 {bone1Name=dyn_jacket_side_r_01, bone2Name=dyn_jacket_side_r_base, swing1LimitDegrees=2.8, swing2LimitDegrees=5, twistLimitDegrees=0}
 {bone1Name=dyn_jacket_bk_r_02, bone2Name=dyn_jacket_bk_r_01, swing1LimitDegrees=5, swing2LimitDegrees=2, twistLimitDegrees=0}
 {bone1Name=dyn_jacket_side_r_02, bone2Name=dyn_jacket_side_r_01, swing1LimitDegrees=2.88, swing2LimitDegrees=5, twistLimitDegrees=0}
 {bone1Name=dyn_jacket_frnt_r_02, bone2Name=dyn_jacket_frnt_r_01, swing1LimitDegrees=27.72, swing2LimitDegrees=5, twistLimitDegrees=10}
 {bone1Name=dyn_jacket_frnt_r_01, bone2Name=dyn_jacket_frnt_r_base, swing1LimitDegrees=39.4, swing2LimitDegrees=5, twistLimitDegrees=5}
]
