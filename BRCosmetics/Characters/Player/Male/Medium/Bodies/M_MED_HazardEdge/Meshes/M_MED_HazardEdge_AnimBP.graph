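# Post-process style AnimBP for the male HazardEdge cosmetic body, adding secondary motion to the pose copied from the attached parent mesh.
# A Control Rig feeds two belt-and-jacket RigidBody variants, selected by front-end state.
# Weighted bone constraints then position straps and zippers, before chained AnimDynamics moves laces, zippers, ears and collar.
# Final bone tweaks compensate for LOD2 and disabled rigid bodies, using flags updated alongside a state-based gravity override.
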AnimBlueprint /BRCosmetics/Characters/Player/Male/Medium/Bodies/M_MED_HazardEdge/Meshes/M_MED_HazardEdge_AnimBP
target_skeleton=&/M_MED_HazardEdge_Skeleton
graphs [
 AnimGraph {nodes [
   Anim Dynamics - Bone: dyn_collar_bk_mid_01!(ComponentPose, Alpha=0, GravityOverride:Vector="0,0,0", ExternalForce="0,0,0") -> (Pose) {
    angularDampingOverride=0.9999
    angularSpringConstant=300
    bAngularSpring=true
    bOverrideAngularDamping=true
    bUseGravityOverride=true
    simulationSpace=BoneRelative
    alphaBoolBlend {blendOption=Linear}
    boundBone {boneName=dyn_collar_bk_mid_01}
    physicsBodyDefinitions [{
      boundBone {boneName=dyn_collar_bk_mid_01}
      boxExtents=<4 4 4>
      constraintSetup {angularLimitsMax=<0 0 2>, angularLimitsMin=<0 0 -2>, angularTarget=<1 0 0>}
      localJointOffset=<0.1 0 0>
     }]
    relativeSpaceBone {boneName=dyn_collar_bk_mid_base}
   }
   n1 = "Transform (Modify) Bone - Bone: LOD2_jacket_support"(
    ComponentPose=Local To Component(LocalPose=Use cached pose 'toTransform'())
    Alpha=1
    Translation="0,0,0"
    Rotation="0,0,0"
    Scale="1,0.6,0.6"
   ) {scaleMode=BMM_Additive, alphaBoolBlend {blendOption=Linear}, boneToModify {boneName=LOD2_jacket_support}, scale {y=0.6, z=0.6}}
   n2 = "Transform (Modify) Bone - Bone: LOD2_support"!(ComponentPose=@n1, bAlphaBoolEnabled:bool=Get Is LOD2() -> (Is LOD2:bool), Translation="12,3.094,0", Rotation="0,0,0", Scale="1,1,1") -> (Pose) {
    alphaInputType=Bool
    rotationSpace=BCS_BoneSpace
    scaleMode=BMM_Additive
    scaleSpace=BCS_BoneSpace
    translationMode=BMM_Additive
    translationSpace=BCS_BoneSpace
    alphaBoolBlend {blendOption=Linear}
    boneToModify {boneName=LOD2_support}
    translation=<12 3.094>
   }
   n3 = "Transform (Modify) Bone - Bone: LOD2_jacket_support"!(
    ComponentPose=@n2
    bAlphaBoolEnabled:bool=Get IsRigidBodyOff() -> (IsRigidBodyOff:bool)
    Translation="0,0,0"
    Rotation="0,0,0"
    Scale="1,0.2,0.2"
   ) -> (Pose) {
    alphaInputType=Bool
    rotationSpace=BCS_BoneSpace
    scaleMode=BMM_Additive
    scaleSpace=BCS_BoneSpace
    translationMode=BMM_Additive
    translationSpace=BCS_BoneSpace
    alphaBoolBlend {blendOption=Linear}
    boneToModify {boneName=LOD2_jacket_support}
    scale {y=0.2, z=0.2}
   }
   Output Pose(Result=Component To Local(ComponentPose=@n3))
   n4 = RigidBody!(ComponentPose=Local To Component(LocalPose=Use cached pose 'Pre sim'()), Alpha, OverrideWorldGravity:Vector="0,0,0", ExternalForce) -> (Pose) {
    bClampLinearTranslationLimitToRefPose=true
    bOverrideWorldGravity=true
    overridePhysicsAsset=&/M_MED_HazardEdge_Physics_Belt
    simulationSpace=BaseBoneSpace
    alphaBoolBlend {blendOption=Linear}
    baseBoneRef {boneName=pelvis}
    simSpaceSettings {
     maxAngularAcceleration=500
     maxAngularVelocity=3
     maxLinearAcceleration=400
     maxLinearVelocity=500
     velocityScaleZ=0
     worldAlpha=1
     externalLinearDragV=<0.1 0.1>
    }
   }
   n5 = RigidBody(ComponentPose=@n4, OverrideWorldGravity="0,0,0") {
    bClampLinearTranslationLimitToRefPose=true
    bOverrideWorldGravity=true
    overridePhysicsAsset=&/M_MED_HazardEdge_Physics_Jacket
    simulationSpace=BaseBoneSpace
    alphaBoolBlend {blendOption=Linear}
    baseBoneRef {boneName=spine_04}
    simSpaceSettings {
     maxAngularAcceleration=200
     maxAngularVelocity=3
     maxLinearAcceleration=150
     maxLinearVelocity=100
     velocityScaleZ=0
     worldAlpha=1
     externalLinearDragV=<0.1 0.1>
    }
   }
   n6 = Component To Local(ComponentPose=@n5)
   n7 = RigidBody(ComponentPose=Local To Component(LocalPose=Use cached pose 'Pre sim'()), OverrideWorldGravity="0,0,0") {
    bClampLinearTranslationLimitToRefPose=true
    bOverrideWorldGravity=true
    overridePhysicsAsset=&M_MED_HazardEdge_Physics_Belt
    simulationSpace=BaseBoneSpace
    alphaBoolBlend {blendOption=Linear}
    baseBoneRef {boneName=pelvis}
    simSpaceSettings {maxAngularAcceleration=150, maxAngularVelocity=3, maxLinearAcceleration=200, maxLinearVelocity=500, velocityScaleZ=0, worldAlpha=0.4936}
   }
   n8 = RigidBody(ComponentPose=@n7, OverrideWorldGravity="0,0,0") {
    bClampLinearTranslationLimitToRefPose=true
    bOverrideWorldGravity=true
    overridePhysicsAsset=&M_MED_HazardEdge_Physics_Jacket
    simulationSpace=BaseBoneSpace
    alphaBoolBlend {blendOption=Linear}
    baseBoneRef {boneName=spine_04}
    simSpaceSettings {maxAngularAcceleration=200, maxAngularVelocity=3, maxLinearAcceleration=150, maxLinearVelocity=100, velocityScaleZ=0, worldAlpha=0.2}
   }
   n9 = Component To Local(ComponentPose=@n8)
   n10 = Get bIsInFrontEnd()
   n11 = "Constraint - dyn_dog_r_base (PPPPPP)"!(
    ComponentPose=Local To Component(LocalPose=Blend Poses by bool(BlendPose_0=@n9, BlendPose_1=@n6, bActiveValue=@n10) {blendType=Linear})
    Alpha=1
    ConstraintWeights_0=0.0648
    ConstraintWeights_1:float=0.0147
    ConstraintWeights_2:float=0.3599
    ConstraintWeights_3:float=0.0127
    ConstraintWeights_4:float=0.2791
    ConstraintWeights_5:float=0.2688
   ) -> (Pose) {
    alphaBoolBlend {blendOption=Linear}
    boneToModify {boneName=dyn_dog_r_base}
    constraintSetup [
     {targetBone {boneName=spine_05}}
     {targetBone {boneName=clavicle_r}}
     {targetBone {boneName=upperarm_r}}
     {targetBone {boneName=deform_elbow_out_r}}
     {targetBone {boneName=upperarm_twist_01_r}}
     {targetBone {boneName=upperarm_twist_02_r}}
    ]
    constraintWeights [1,1,1,1,1,1]
   }
   n12 = "Constraint - dyn_dog_l_base (PPPPP)"!(
    ComponentPose=@n11
    Alpha=1
    ConstraintWeights_0=0.052
    ConstraintWeights_1:float=0.0192
    ConstraintWeights_2:float=0.5259
    ConstraintWeights_3:float=0.2012
    ConstraintWeights_4:float=0.2017
   ) -> (Pose) {
    alphaBoolBlend {blendOption=Linear}
    boneToModify {boneName=dyn_dog_l_base}
    constraintSetup [
     {targetBone {boneName=spine_05}}
     {targetBone {boneName=clavicle_l}}
     {targetBone {boneName=upperarm_l}}
     {targetBone {boneName=upperarm_twist_01_l}}
     {targetBone {boneName=upperarm_twist_02_l}}
    ]
    constraintWeights [1,1,1,1,1]
   }
   n13 = "Constraint - dyn_zippo_l (PPPPPPPP)"!(
    ComponentPose=@n12
    Alpha=1
    ConstraintWeights_0=0.277
    ConstraintWeights_1:float=0.0004
    ConstraintWeights_2:float=0.0001
    ConstraintWeights_3:float=0
    ConstraintWeights_4:float=0
    ConstraintWeights_5:float=0
    ConstraintWeights_6:float=0.7225
    ConstraintWeights_7:float=0
   ) -> (Pose) {
    alphaBoolBlend {blendOption=Linear}
    boneToModify {boneName=dyn_zippo_l}
    constraintSetup [
     {targetBone {boneName=spine_03}}
     {targetBone {boneName=spine_04}}
     {targetBone {boneName=spine_05}}
     {targetBone {boneName=clavicle_l}}
     {targetBone {boneName=deform_pec_l}}
     {targetBone {boneName=dyn_jacket_frnt_l_base}}
     {targetBone {boneName=dyn_jacket_frnt_l_01}}
     {targetBone {boneName=dyn_jacket_frnt_l_02}}
    ]
    constraintWeights [1,1,1,1,1,1,1,1]
   }
   n14 = "Constraint - dyn_zippo_r_01 (PPPPPPPP)"!(
    ComponentPose=@n13
    Alpha=1
    ConstraintWeights_0=0.1753
    ConstraintWeights_1:float=0.0542
    ConstraintWeights_2:float=0.0006
    ConstraintWeights_3:float=0.7547
    ConstraintWeights_4:float=0.0081
    ConstraintWeights_5:float=0.0007
    ConstraintWeights_6:float=0.0062
    ConstraintWeights_7:float=0.0001
   ) -> (Pose) {
    alphaBoolBlend {blendOption=Linear}
    boneToModify {boneName=dyn_zippo_r_01}
    constraintSetup [
     {targetBone {boneName=spine_03}}
     {targetBone {boneName=spine_04}}
     {targetBone {boneName=spine_05}}
     {targetBone {boneName=dyn_jacket_frnt_r_01}}
     {targetBone {boneName=dyn_jacket_frnt_r_02}}
     {targetBone {boneName=dyn_jacket_side_r_01}}
     {targetBone {boneName=dyn_jacket_side_r_02}}
     {targetBone {boneName=dyn_jacket_bk_r_02}}
    ]
    constraintWeights [1,1,1,1,1,1,1,1]
   }
   n15 = "Constraint - dyn_zippo_r_02 (P)"!(ComponentPose=@n14, Alpha=1, ConstraintWeights_0=1, ConstraintWeights_1:float=0) -> (Pose) {
    alphaBoolBlend {blendOption=Linear}
    boneToModify {boneName=dyn_zippo_r_02}
    constraintSetup [{targetBone {boneName=dyn_jacket_frnt_r_02}}]
    constraintWeights [1,0]
   }
   Save cached pose 'toAnimDyn1'(Pose=Component To Local(ComponentPose=@n15)) {cacheName=toAnimDyn1}
   n16 = Anim Dynamics - Bone: dyn_lace_l_01!(ComponentPose=Local To Component(LocalPose=Use cached pose 'toAnimDyn1'()), Alpha=1, GravityOverride:Vector="0,0,0", ExternalForce="0,0,0") -> (Pose) {
    angularDampingOverride=1
    angularSpringConstant=200
    bAngularSpring=true
    bLinearSpring=true
    bOverrideAngularDamping=true
    bUseGravityOverride=true
    alphaBoolBlend {blendOption=Linear}
    boundBone {boneName=dyn_lace_l_01}
    componentLinearAccScale=<1 1 1>
    componentLinearVelScale=<1 1 1>
    physicsBodyDefinitions [{
      boundBone {boneName=dyn_lace_l_01}
      boxExtents=<4 4 4>
      constraintSetup {angularLimitsMax=<5 5 20>, angularLimitsMin=<-5 -5 0>, angularTarget=<1 0 0>}
      localJointOffset=<-0.2 0 0>
     }]
    relativeSpaceBone {boneName=dyn_lace_l_base}
   }
   n17 = Anim Dynamics - Bone: dyn_lace_l_02!(ComponentPose=@n16, Alpha=1, GravityOverride:Vector="0,0,0", ExternalForce="0,0,0") -> (Pose) {
    angularDampingOverride=1
    angularSpringConstant=200
    bAngularSpring=true
    bLinearSpring=true
    bOverrideAngularDamping=true
    bUseGravityOverride=true
    alphaBoolBlend {blendOption=Linear}
    boundBone {boneName=dyn_lace_l_02}
    componentLinearAccScale=<1 1 1>
    componentLinearVelScale=<1 1 1>
    physicsBodyDefinitions [{
      boundBone {boneName=dyn_lace_l_02}
      boxExtents=<4 4 4>
      constraintSetup {angularLimitsMax=<5 5 20>, angularLimitsMin=<-5 -5 0>, angularTarget=<1 0 0>}
      localJointOffset=<-0.2 0 0>
     }]
    relativeSpaceBone {boneName=dyn_lace_l_base}
   }
   n18 = Anim Dynamics - Bone: dyn_lace_r_01!(ComponentPose=@n17, Alpha=1, GravityOverride:Vector="0,0,0", ExternalForce="0,0,0") -> (Pose) {
    angularDampingOverride=1
    angularSpringConstant=200
    bAngularSpring=true
    bLinearSpring=true
    bOverrideAngularDamping=true
    bUseGravityOverride=true
    alphaBoolBlend {blendOption=Linear}
    boundBone {boneName=dyn_lace_r_01}
    componentLinearAccScale=<1 1 1>
    componentLinearVelScale=<1 1 1>
    physicsBodyDefinitions [{
      boundBone {boneName=dyn_lace_r_01}
      boxExtents=<4 4 4>
      constraintSetup {angularLimitsMax=<5 10 20>, angularLimitsMin=<-5 -10 0>, angularTarget=<1 0 0>}
      localJointOffset=<0.2 0 0>
     }]
    relativeSpaceBone {boneName=dyn_lace_r_base}
   }
   n19 = Anim Dynamics - Bone: dyn_lace_r_02!(ComponentPose=@n18, Alpha=1, GravityOverride:Vector="0,0,0", ExternalForce="0,0,0") -> (Pose) {
    angularDampingOverride=1
    angularSpringConstant=200
    bAngularSpring=true
    bLinearSpring=true
    bOverrideAngularDamping=true
    bUseGravityOverride=true
    alphaBoolBlend {blendOption=Linear}
    boundBone {boneName=dyn_lace_r_02}
    componentLinearAccScale=<1 1 1>
    componentLinearVelScale=<1 1 1>
    physicsBodyDefinitions [{
      boundBone {boneName=dyn_lace_r_02}
      boxExtents=<4 4 4>
      constraintSetup {angularLimitsMax=<5 10 20>, angularLimitsMin=<-5 -10 0>, angularTarget=<1 0 0>}
      localJointOffset=<0.2 0 0>
     }]
    relativeSpaceBone {boneName=foot_r}
   }
   n20 = Anim Dynamics - Bone: dyn_zippo_r_01!(ComponentPose=@n19, Alpha=1, GravityOverride:Vector="0,0,0", ExternalForce="0,0,0") -> (Pose) {
    angularDampingOverride=1
    bOverrideAngularDamping=true
    bUseGravityOverride=true
    alphaBoolBlend {blendOption=Linear}
    boundBone {boneName=dyn_zippo_r_01}
    physicsBodyDefinitions [{
      boundBone {boneName=dyn_zippo_r_01}
      boxExtents=<2 2 2>
      constraintSetup {angularLimitsMax=<10 20 20>, angularLimitsMin=<-10 -20 -20>, angularTarget=<1 0 0>}
      localJointOffset=<0.1 0 0>
     }]
    planarLimits [{drivingBone {boneName=dyn_zippo_r_01}, planeTransform {location=<0 0 0>, rotation=<rot -0.0131 0.0293 67.5188>}}]
    relativeSpaceBone {boneName=dyn_jacket_frnt_r_01}
   }
   n21 = Anim Dynamics - Bone: dyn_zippo_r_02!(ComponentPose=@n20, Alpha=1, GravityOverride:Vector="0,0,0", ExternalForce="0,0,0") -> (Pose) {
    angularDampingOverride=1
    bOverrideAngularDamping=true
    bUseGravityOverride=true
    alphaBoolBlend {blendOption=Linear}
    boundBone {boneName=dyn_zippo_r_02}
    physicsBodyDefinitions [{
      boundBone {boneName=dyn_zippo_r_02}
      boxExtents=<2 2 2>
      constraintSetup {angularLimitsMax=<10 20 20>, angularLimitsMin=<-10 -20 -20>, angularTarget=<1 0 0>}
      localJointOffset=<0.1 0 0>
     }]
    planarLimits [{drivingBone {boneName=dyn_zippo_r_02}, planeTransform {location=<0 0 0>, rotation=<rot -0.0306 0.0182 103.3232>}}]
    relativeSpaceBone {boneName=dyn_jacket_frnt_r_02}
   }
   n22 = Anim Dynamics - Bone: dyn_zippo_l!(ComponentPose=@n21, Alpha=1, GravityOverride:Vector="0,0,0", ExternalForce="0,0,0") -> (Pose) {
    angularDampingOverride=1
    bOverrideAngularDamping=true
    bUseGravityOverride=true
    alphaBoolBlend {blendOption=Linear}
    boundBone {boneName=dyn_zippo_l}
    physicsBodyDefinitions [{
      boundBone {boneName=dyn_zippo_l}
      boxExtents=<2 2 2>
      constraintSetup {angularLimitsMax=<10 10 20>, angularLimitsMin=<-10 -10 -20>, angularTarget=<1 0 0>}
      localJointOffset=<0.1 0 0>
     }]
    planarLimits [{drivingBone {boneName=dyn_zippo_l}, planeTransform {location=<0 0 0>, rotation=<rot -0.0127 0.0205 17.7245>}}]
    relativeSpaceBone {boneName=dyn_jacket_frnt_l_01}
   }
   Save cached pose 'toAnimDyn2'(Pose=Component To Local(ComponentPose=@n22)) {cacheName=toAnimDyn2}
   n23 = Anim Dynamics - Bone: dyn_dog_ear_l_01!(ComponentPose=Local To Component(LocalPose=Use cached pose 'toAnimDyn2'()), Alpha=1, GravityOverride:Vector="0,0,0", ExternalForce="0,0,0") -> (Pose) {
    angularDampingOverride=0.9999
    angularSpringConstant=500
    bAngularSpring=true
    bOverrideAngularDamping=true
    bUseGravityOverride=true
    alphaBoolBlend {blendOption=Linear}
    boundBone {boneName=dyn_dog_ear_l_01}
    physicsBodyDefinitions [{
      boundBone {boneName=dyn_dog_ear_l_01}
      boxExtents=<3 3 3>
      constraintSetup {angularLimitsMax=<20 20 20>, angularLimitsMin=<-20 -20 -20>, angularTarget=<1 0 0>}
      localJointOffset=<0.1 0 0>
     }]
    relativeSpaceBone {boneName=spine_05}
   }
   n24 = Anim Dynamics - Bone: dyn_dog_ear_l_02!(ComponentPose=@n23, Alpha=1, GravityOverride:Vector="0,0,0", ExternalForce="0,0,0") -> (Pose) {
    angularDampingOverride=0.9999
    angularSpringConstant=500
    bAngularSpring=true
    bOverrideAngularDamping=true
    bUseGravityOverride=true
    alphaBoolBlend {blendOption=Linear}
    boundBone {boneName=dyn_dog_ear_l_02}
    physicsBodyDefinitions [{
      boundBone {boneName=dyn_dog_ear_l_02}
      boxExtents=<3 3 3>
      constraintSetup {angularLimitsMax=<20 20 20>, angularLimitsMin=<-20 -20 -20>, angularTarget=<1 0 0>}
      localJointOffset=<0.1 0 0>
     }]
    relativeSpaceBone {boneName=spine_05}
   }
   n25 = Anim Dynamics - Bone: dyn_dog_ear_r_01!(ComponentPose=@n24, Alpha=1, GravityOverride:Vector="0,0,0", ExternalForce="0,0,0") -> (Pose) {
    angularDampingOverride=0.9999
    angularSpringConstant=500
    bAngularSpring=true
    bOverrideAngularDamping=true
    bUseGravityOverride=true
    alphaBoolBlend {blendOption=Linear}
    boundBone {boneName=dyn_dog_ear_r_01}
    physicsBodyDefinitions [{
      boundBone {boneName=dyn_dog_ear_r_01}
      boxExtents=<3 3 3>
      constraintSetup {angularLimitsMax=<20 20 20>, angularLimitsMin=<-20 -20 -20>, angularTarget=<1 0 0>}
      localJointOffset=<0.1 0 0>
     }]
    relativeSpaceBone {boneName=spine_05}
   }
   n26 = Anim Dynamics - Bone: dyn_dog_ear_r_02!(ComponentPose=@n25, Alpha=1, GravityOverride:Vector="0,0,0", ExternalForce="0,0,0") -> (Pose) {
    angularDampingOverride=0.9999
    angularSpringConstant=500
    bAngularSpring=true
    bOverrideAngularDamping=true
    bUseGravityOverride=true
    alphaBoolBlend {blendOption=Linear}
    boundBone {boneName=dyn_dog_ear_r_02}
    physicsBodyDefinitions [{
      boundBone {boneName=dyn_dog_ear_r_02}
      boxExtents=<3 3 3>
      constraintSetup {angularLimitsMax=<20 20 20>, angularLimitsMin=<-20 -20 -20>, angularTarget=<1 0 0>}
      localJointOffset=<0.1 0 0>
     }]
    relativeSpaceBone {boneName=spine_05}
   }
   n27 = Anim Dynamics - Bone: dyn_collar_l_01!(ComponentPose=@n26, Alpha=1, GravityOverride:Vector="0,0,0", ExternalForce="0,0,0") -> (Pose) {
    angularDampingOverride=0.999
    angularSpringConstant=300
    bAngularSpring=true
    bOverrideAngularDamping=true
    bUseGravityOverride=true
    alphaBoolBlend {blendOption=Linear}
    boundBone {boneName=dyn_collar_l_01}
    physicsBodyDefinitions [{
      boundBone {boneName=dyn_collar_l_01}
      boxExtents=<4 4 4>
      constraintSetup {angularLimitsMax=<0 5 0>, angularLimitsMin=<0 -5 0>, angularTarget=<1 0 0>}
      localJointOffset=<0.1 0 0>
     }]
   }
   n28 = Anim Dynamics - Bone: dyn_collar_r_01!(ComponentPose=@n27, Alpha=1, GravityOverride:Vector="0,0,0", ExternalForce="0,0,0") -> (Pose) {
    angularDampingOverride=0.999
    angularSpringConstant=300
    bAngularSpring=true
    bOverrideAngularDamping=true
    bUseGravityOverride=true
    alphaBoolBlend {blendOption=Linear}
    boundBone {boneName=dyn_collar_r_01}
    physicsBodyDefinitions [{
      boundBone {boneName=dyn_collar_r_01}
      boxExtents=<4 4 4>
      constraintSetup {angularLimitsMax=<0 5 0>, angularLimitsMin=<0 -5 0>, angularTarget=<1 0 0>}
      localJointOffset=<0.1 0 0>
     }]
   }
   Save cached pose 'toTransform'(Pose=Component To Local(ComponentPose=@n28)) {cacheName=toTransform}
   n29 = Control Rig(Source=Copy Pose From Mesh() {bUseAttachedParent=true}) {
    alphaBoolBlend {blendOption=Linear}
    controlRigAssetReference {blueprintRigClass=&/M_MED_HazardEdge_CtrlRig.M_MED_HazardEdge_CtrlRig_C}
   }
   Save cached pose 'Pre sim'(Pose=@n29) {cacheName=Pre sim}
  ]}
 EventGraph {nodes [
   Event BlueprintUpdateAnimation()
   > n1 = Gravity Override State Function DEPRECATED(CharacterPartAnimInstance=Self-Reference(), GameTimeInSeconds=GetGameTimeInSeconds(), GravityOverrideParams=Get Gravity Override Params())
   > Set GravityOverride(GravityOverride=@n1)
   > Set Is LOD2(Is LOD2:bool="integer > integer"(A=Get CurrentLOD(), B=1)) -> (Output_Get:bool)
   > Set IsRigidBodyOff(IsRigidBodyOff:bool=NOT Boolean(A=Get bRigidBodyEnabled())) -> (Output_Get:bool)
   TryGetPawnOwner()
  ]}
]
variables [
 Gravity Override Params {
  type=GravityOverrideParamsStruct
  default {
   emoteOrMelee {gravityOverride=<0 0 -580>}
   skydive_FloatingDown {windAmplitude=<200 -500 50>, windFrequency=<1.5 7 4>, windOffset=<0 -600 100>}
   skydive_Parachuting {gravityOverride=<0 -400 -450>, windAmplitude=<200 -300 400>, windFrequency=<1.5 4 2>}
  }
 }
 GravityOverride {type=Vector}
 Is LOD2 {type=boolean}
 IsRigidBodyOff {type=boolean}
]
