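# Control Rig for the HazardEdge medium male FaceAcc part, applying corrective transforms to hood support bones.
# Sequence output B squashes neck_hood_support, then a spherical pose reader on upperarm_l, with debug draw on, weights further adjustments to neck and left-side hood bones.
# A second chain maps spine_05-to-pelvis rotation to a front-hood weight and offsets the left side, but is not wired to execution; output A ends at a reroute.
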ControlRigBlueprint /BRCosmetics/Characters/Player/Male/Medium/Bodies/M_MED_HazardEdge/Meshes/Parts/M_MED_HazardEdge_FaceAcc_CtrlRig
graphs [
 RigVMFunctionLibrary {links [], nodes []}
 RigVMModel {nodes [
   n1 = &RigUnit_GetRelativeTransformForItem(Child=(Type=Bone,Name="spine_05"), Parent=(Type=Bone,Name="pelvis"), bChildInitial=False, bParentInitial=False)
   AlphaInterp = &RigVMFunction_AlphaInterp(
    InterpSpeedDecreasing=24
    InterpSpeedIncreasing=24
    Scale=1
    bClampResult=False
    bInterpResult=False
    Value=&RigVMFunction_MathDoubleRemap(
     SourceMaximum=-17
     SourceMinimum=10
     TargetMaximum=1
     TargetMinimum=0
     Value=&RigVMFunction_MathQuaternionToEuler(RotationOrder=ZXY, Value=@n1.RelativeTransform.Rotation).Result.Z
    ).Result
   )
   &RigUnit_ModifyTransforms(
    ExecutePin=()
    ItemToModify=((Item=(Type=Bone,Name="hood_support_frnt"),Transform=(Rotation=(X=0,Y=0,Z=0,W=-1),Translation=(X=-9.2794,Y=0,Z=0.9772),Scale3D=(X=1,Y=1.3,Z=1))))
    WeightMinimum=0
    Weight=@AlphaInterp.Result
   )
   > &RigUnit_ModifyTransforms(
    ItemToModify=((Item=(Type=Bone,Name="support_hood_side_l"),Transform=(Rotation=(X=0,Y=0,Z=0,W=-1),Translation=(X=-1.4628,Y=2.6005,Z=1.1628),Scale3D=(X=1,Y=1,Z=1))))
    Weight=1
    WeightMinimum=0
   )
   RigUnit_BeginExecution = &RigUnit_BeginExecution()
   RigVMFunction_Sequence = &RigVMFunction_Sequence(ExecuteContext=@RigUnit_BeginExecution.ExecutePin)
   &RigUnit_ModifyTransforms(
    ItemToModify=((Item=(Type=Bone,Name="neck_hood_support"),Transform=(Rotation=(X=0,Y=0,Z=0,W=1),Translation=(X=0,Y=0,Z=0),Scale3D=(X=1,Y=0.7,Z=0.9))))
    Weight=1
    WeightMinimum=0
    ExecutePin=@RigVMFunction_Sequence.B
   )
   > n2 = &RigUnit_SphericalPoseReader(
    ActiveRegionScaleFactors=(PositiveWidth=1,NegativeWidth=1,PositiveHeight=1,NegativeHeight=1)
    ActiveRegionSize=0.1
    Debug=(bDrawDebug=True,bDraw2D=False,bDrawLocalAxes=False,DebugScale=25,DebugSegments=20,DebugThickness=0.25)
    DriverAxis=(X=1,Y=0,Z=0)
    DriverItem=(Type=Bone,Name="upperarm_l")
    FalloffRegionScaleFactors=(PositiveWidth=1.5,NegativeWidth=1.5,PositiveHeight=2.5,NegativeHeight=1.5)
    FalloffSize=0.2
    FlipHeightScaling=False
    FlipWidthScaling=False
    OptionalParentItem=(Type=Bone,Name="None")
    RotationOffset=(X=25,Y=-45,Z=0)
   )
   > &RigUnit_ModifyTransforms(
    ItemToModify="((Item=(Type=Bone,Name=\"neck_hood_support\"),Transform=(Rotation=(X=0,Y=0,Z=0,W=1),Translation=(X=0,Y=-5,Z=0),Scale3D=(X=1,Y=0.6,Z=0.7))),(Item=(Type=Bone,Name=\"support_hood_side_l\"),Transform=(Rotation=(X=0,Y=0,Z=0,W=1),Translation=(X=0,Y=0,Z=0),Scale3D=(X ...(truncated)"
    WeightMinimum=0
    Weight=@n2.OutputParam
   )
   RigVMRerouteNode(Value=@RigVMFunction_Sequence.A)
  ]}
]
hierarchy [
 root {loc=<0 0 0>, rot=<0 0 0 1>}
 pelvis {parent=root, loc=<0 -0.7071 93.1191>, rot=<0.0127 0.707 -0.0127 -0.707>}
 spine_01 {parent=pelvis, loc=<6.7494 -0.6388 0>, rot=<0 0 0.034 -0.9994>}
 spine_02 {parent=spine_01, loc=<7.2518 -0.035 0>, rot=<0 0 0.0238 -0.9997>}
 spine_03 {parent=spine_02, loc=<6.0104 0 0>, rot=<0 0 0.0149 -0.9999>}
 spine_04 {parent=spine_03, loc=<8.1024 0 0>, rot=<0 0 -0.0132 -0.9999>}
 spine_05 {parent=spine_04, loc=<10.0146 0 0>, rot=<0 0 -0.0015 -1>}
 clavicle_l {parent=spine_05, loc=<4.3575 -0.1466 -2.3223>, rot=<0.0465 -0.7289 -0.009 -0.6829>}
 dyn_collar_frnt_l_base {parent=clavicle_l, loc=<3.939 13.4232 -0.6341>, rot=<-0.2445 -0.0386 -0.2564 -0.9343>}
 dyn_collar_frnt_l_1 {parent=dyn_collar_frnt_l_base, loc=<0 0 0>, rot=<0 0 0 1>}
 dyn_collar_l_base {parent=clavicle_l, loc=<8.0978 0.8085 7.9677>, rot=<0.0713 0.103 -0.0178 -0.992>}
 dyn_collar_l_01 {parent=dyn_collar_l_base, loc=<0 0 0>, rot=<0 0 0 1>}
 upperarm_l {parent=clavicle_l, loc=<15.5111 0 0>, rot=<0.0226 -0.3234 -0.0088 -0.9459>}
 neck_01 {parent=spine_05, loc=<11.3138 0.1529 0>, rot=<0 0 -0.2003 -0.9797>}
 neck_02 {parent=neck_01, loc=<4.5627 0 0>, rot=<0 0 0.0504 -0.9987>}
 head {parent=neck_02, loc=<5.3914 0 0>, rot=<0 0 0.115 -0.9934>}
 support_hood_side_l {parent=head, loc=<-1.2007 -3.0384 -9.3909>, rot=<0.0441 -0.7057 0.0383 -0.7061>}
 support_hood_bk_mid {parent=head, loc=<-0.2542 -9.3903 0>, rot=<0.0029 -0.7071 -0.0029 -0.7071>}
 support_hood_side_r {parent=head, loc=<-1.2012 -3.0384 9.3909>, rot=<-0.7057 -0.0441 0.7061 0.0383>}
 hood_support_frnt {parent=head, loc=<-10.7045 12.4989 -1.2324>, rot=<0.7037 0.0921 -0.0937 -0.6983>}
 hood_support_top {parent=head, loc=<12.8428 5.8199 0>, rot=<0.0029 -0.7071 -0.0029 -0.7071>}
 joint1 {parent=head, loc=<12.8428 5.8199 0>, rot=<0.0029 -0.7071 -0.0029 -0.7071>}
 faceAttach {parent=head, loc=<1.2261 0.4344 0>, rot=<0.0082 -0.7071 -0.0082 -0.7071>, scale=<1.1633 1.1633 1.1633>}
 C_brow_mid {parent=faceAttach, loc=<0.0719 8.0601 8.387>, rot=<-0.2564 -0.2298 -0.0165 -0.9387>}
 L_brow_mid {parent=faceAttach, loc=<3.0283 7.575 8.1201>, rot=<-0.2138 -0.0333 0.2593 -0.9413>}
 L_brow_outer {parent=faceAttach, loc=<6.2141 6.7224 6.0346>, rot=<-0.1243 -0.1323 0.4575 -0.8705>}
 L_cheek_inner {parent=faceAttach, loc=<6.081 9.0609 -0.7128>, rot=<-0.1604 -0.5478 0.232 -0.7877>}
 C_lip_upper_mid {parent=faceAttach, loc=<-0.0094 11.5446 1.865>, rot=<0.2209 0.1117 -0.0419 -0.968>}
 L_lip_upper_outer {parent=faceAttach, loc=<2.5826 11.2866 0.819>, rot=<-0.1207 -0.6183 0.3799 -0.6773>}
 L_lip_corner {parent=faceAttach, loc=<3.1035 9.7801 -3.78>, rot=<-0.1065 0.1456 0.1349 -0.9743>}
 C_nose_bridge {parent=faceAttach, loc=<-0.0111 10.9453 5.3047>, rot=<-0.223 -0.008 -0.0023 -0.9748>}
 R_brow_mid {parent=faceAttach, loc=<-3.0283 7.575 8.1201>, rot=<-0.2138 0.0333 -0.2593 -0.9413>}
 R_brow_outer {parent=faceAttach, loc=<-6.2141 6.7224 6.0346>, rot=<-0.1243 0.1323 -0.4575 -0.8705>}
 R_cheek_inner {parent=faceAttach, loc=<-6.081 9.0609 -0.7128>, rot=<-0.1604 0.5478 -0.232 -0.7877>}
 R_lip_corner {parent=faceAttach, loc=<-3.1035 9.7801 -3.78>, rot=<-0.1065 -0.1456 -0.1349 -0.9743>}
 C_jaw {parent=faceAttach, loc=<-0.0151 0.778 -2.4728>, rot=<0 0 0 1>}
 L_lip_lower_outer {parent=C_jaw, loc=<1.9905 9.6334 1.0808>, rot=<-0.0247 -0.1415 0.1039 -0.9842>}
 C_lip_lower_mid {parent=C_jaw, loc=<-0.0534 10.4178 3.4504>, rot=<0 0.098 0.0208 -0.995>}
 R_lip_lower_outer {parent=C_jaw, loc=<-1.9604 9.6334 1.0808>, rot=<-0.0247 0.1415 -0.1039 -0.9842>}
 teeth_lower {parent=C_jaw, loc=<0.0151 5.8762 -1.2173>, rot=<0.0217 0 0 -0.9998>}
 tongue {parent=C_jaw, loc=<0.0158 2.4627 0.4653>, rot=<-0.1568 0 0 -0.9876>}
 R_eye {parent=faceAttach, loc=<-4.4835 6.4222 4.442>, rot=<0 0 0.9998 -0.0187>, scale=<2.055 2.055 2.055>}
 R_eye_lid_upper_mid {parent=R_eye, loc=<-0.002 -1.1718 0.263>, rot=<0.0523 0 0.0007 -0.9986>, scale=<0.9998 0.9998 0.9998>}
 R_eye_lid_lower_mid {parent=R_eye, loc=<0.0219 -0.9421 -0.705>, rot=<0.0329 0 0.0007 -0.9995>, scale=<0.9998 0.9998 0.9998>}
 L_eye {parent=faceAttach, loc=<4.4835 6.4222 4.442>, rot=<0 0 0.9998 0.0187>, scale=<2.055 2.055 2.055>}
 L_eye_lid_upper_mid {parent=L_eye, loc=<0.0251 -1.1621 0.2388>, rot=<0.0523 0 0 -0.9986>, scale=<0.9998 0.9998 0.9998>}
 L_eye_lid_lower_mid {parent=L_eye, loc=<0.0223 -0.9228 -0.7486>, rot=<0.0329 0 0 -0.9995>, scale=<0.9998 0.9998 0.9998>}
 teeth_upper {parent=faceAttach, loc=<0 7.3065 -1.2818>, rot=<0 0 0 1>}
 R_lip_upper_outer {parent=faceAttach, loc=<-2.5826 11.2866 0.819>, rot=<-0.1207 0.6183 -0.3799 -0.6773>}
 neck_hood_support {parent=neck_01, loc=<0 0 0>, rot=<0 0 0 1>}
 clavicle_r {parent=spine_05, loc=<4.3575 -0.1466 2.3223>, rot=<0.7289 0.0465 -0.6829 0.009>}
 dyn_collar_frnt_r_base {parent=clavicle_r, loc=<-3.4646 -13.1503 0.663>, rot=<-0.2445 -0.0386 -0.2564 -0.9343>}
 dyn_collar_frnt_r_1 {parent=dyn_collar_frnt_r_base, loc=<0 0 0>, rot=<0 0 0 1>}
 dyn_collar_r_base {parent=clavicle_r, loc=<-8.0978 -0.8085 -7.9675>, rot=<0.0713 0.103 -0.0178 -0.992>}
 dyn_collar_r_01 {parent=dyn_collar_r_base, loc=<0 0 0>, rot=<0 0 0 1>}
 upperarm_r {parent=clavicle_r, loc=<-15.5111 0 0>, rot=<0.0226 -0.3234 -0.0088 -0.9459>}
 dyn_collar_bk_mid_base {parent=spine_05, loc=<13.1626 -6.6507 0>, rot=<0 0 0.4025 -0.9154>}
 dyn_collar_bk_mid_01 {parent=dyn_collar_bk_mid_base, loc=<0 0 0>, rot=<0 0 0 1>}
 deform_pec_aim_base_l {parent=spine_04, loc=<7.8732 12.6388 -2.4123>, rot=<0.0102 -0.4602 0.3979 -0.7936>}
 deform_pec_aim_l {parent=deform_pec_aim_base_l, loc=<0 0 0>, rot=<0 0 0 1>}
 deform_pec_base_l {parent=deform_pec_aim_l, loc=<10.766 0 0>, rot=<0 0 0 1>}
 deform_pec_l {parent=deform_pec_base_l, loc=<0.055 0 0>, rot=<0 0 0 1>}
 deform_pec_aim_base_r {parent=spine_04, loc=<7.8727 12.6388 2.4123>, rot=<-0.4602 -0.0102 0.7936 0.3979>}
 deform_pec_aim_r {parent=deform_pec_aim_base_r, loc=<0 0 0>, rot=<0 0 0 1>}
 deform_pec_base_r {parent=deform_pec_aim_r, loc=<-10.7662 -0.0006 -0.0007>, rot=<0 0 0 1>}
 deform_pec_r {parent=deform_pec_base_r, loc=<0 0 0>, rot=<0 0 0 1>}
]
