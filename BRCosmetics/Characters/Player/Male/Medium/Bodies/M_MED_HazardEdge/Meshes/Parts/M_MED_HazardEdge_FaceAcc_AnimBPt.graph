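# Animation Blueprint for the HazardEdge medium male face accessory, posing it from the character's other meshes.
# The pose copied from the attached parent is layered with one copied from the Body mesh, filtered to the clavicle branches.
# A face accessory Control Rig then processes the blended pose.
# The Body reference is taken from the custom part array when character customization finishes loading.
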AnimBlueprint /BRCosmetics/Characters/Player/Male/Medium/Bodies/M_MED_HazardEdge/Meshes/Parts/M_MED_HazardEdge_FaceAcc_AnimBPt
target_skeleton=&/M_MED_HazardEdge_FaceAcc_Skeleton
graphs [
 AnimGraph {nodes [
   n1 = Layered blend per bone(
    BasePose=Copy Pose From Mesh() {bUseAttachedParent=true}
    BlendPoses_0=Copy Pose From Mesh(SourceMeshComponent=Get Body() -> (Body:Skeletal Mesh Component Object Reference))
   ) {layerSetup [{branchFilters [{blendDepth=2, boneName=clavicle_r}, {blendDepth=2, boneName=clavicle_l}, {boneName=None}, {boneName=None}]}]}
   n2 = Control Rig(Source=@n1) {
    alphaBoolBlend {blendOption=Linear}
    controlRigAssetReference {blueprintRigClass=&/M_MED_HazardEdge_FaceAcc_CtrlRig.M_MED_HazardEdge_FaceAcc_CtrlRig_C}
   }
   Output Pose(Result=@n2)
  ]}
 EventGraph {nodes [
   n1 = Event OnCharacterCustomizationLoaded()
   > Set Body(Body:Skeletal Mesh Component Object Reference=Get Array Item with EFort Custom Part Type(Array=@n1.CustomMeshParts)) -> (Output_Get:Skeletal Mesh Component Object Reference)
   TryGetPawnOwner()
   Event BlueprintUpdateAnimation()
  ]}
]
variables [Body {type=SkeletalMeshComponent}]
